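# Physics asset for the FuryFax backpack cosmetic, defining capsule bodies and joints for the pack and its dangling strap.
# Two large capsules approximate the pack, while small capsules cover the strap base and four dynamic strap bones.
# The strap bones form a chain from base_strap with angular limits that loosen toward the tip and tighter twist than swing.
# The joint between base_strap and pack leaves swing and twist free.
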PhysicsAsset /BRCosmetics/Accessories/FORT_Backpacks/Backpack_F_MED_FuryFax/Meshes/F_MED_FuryFax_Pack_Physics
bodies {
 base_strap [{shapeType=Capsule, length=0.505, radius=0.505, shapeName=base_strap_capsule, rotation=<rot 0 -90 0>}]
 dyn_strap_01 [{shapeType=Capsule, length=1, radius=0.75, center=<1 0 0>, rotation=<rot 90 0 0>}]
 dyn_strap_02 [{shapeType=Capsule, length=1, radius=0.75, center=<1 0 0>, rotation=<rot 90 0 0>}]
 dyn_strap_03 [{shapeType=Capsule, length=1, radius=0.75, center=<1.8 0 0>, rotation=<rot 90 0 0>}]
 dyn_strap_04 [{shapeType=Capsule, length=2, radius=1, center=<3 0 0>, rotation=<rot 90 0 0>}]
 pack [
  {
   shapeType=Capsule
   length=11.4941
   radius=6.8341
   shapeName=pack_capsule
   center=<-0.1053 -4.2848 -0.0233>
   rotation=<rot 0.0008 92.3024 -46.7502>
  }
  {shapeType=Capsule, length=8, radius=8, shapeName=pack_capsule, center=<9.7323 6.6327 -13.0663>, rotation=<rot 0 0 10>}
 ]
}
constraints [
 {
  bone1Name=base_strap
  bone2Name=pack
  swing1LimitDegrees=45
  swing1Motion=Free
  swing2LimitDegrees=45
  swing2Motion=Free
  twistLimitDegrees=45
  twistMotion=Free
 }
 {bone1Name=dyn_strap_01, bone2Name=base_strap, swing1LimitDegrees=10, swing2LimitDegrees=10, twistLimitDegrees=5}
 {bone1Name=dyn_strap_02, bone2Name=dyn_strap_01, swing1LimitDegrees=15, swing2LimitDegrees=15, twistLimitDegrees=5}
 {bone1Name=dyn_strap_03, bone2Name=dyn_strap_02, swing1LimitDegrees=30, swing2LimitDegrees=20, twistLimitDegrees=8}
 {bone1Name=dyn_strap_04, bone2Name=dyn_strap_03, swing1LimitDegrees=30, swing2LimitDegrees=15, twistLimitDegrees=8}
]
